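# Cosmetic Niagara effect for the PunkDevil gold skin that scatters small, warm gold sparkles along the character's two horns.
# Four cone locations sample tip and base segments of each horn, and random A-or-B lerps choose one per particle.
# Skeletal mesh sampling each update keeps that offset attached to the character, while curves fade and scale each spinning sprite with skewed random brightness flares.
# The NE_Empty sprite and mesh emitter is disabled.
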NiagaraSystem /BRCosmetics/Characters/Player/Female/Medium/Bodies/F_MED_Punk_Devil/Skins/PunkDevil_Gold/Effects/NS_Skin_PunkDevil_ExtremeGold
dependencies {
 usedDataInterfaces [&NiagaraDataInterfaceCurve]
 usedDynamicInputs []
 usedModules [
  &SystemState
  &self:SystemSpawnScript.SystemScriptSource.SystemScriptGraph.NiagaraNodeAssignment_4.SetVariables_#1
 ]
}
emitters {
 NE_Empty {
  bEnabled=false
  emitterUpdateScript [
   &EmitterState(Life Cycle Mode=::NewEnumerator1, Loop Duration=1, UseLoopDelay=false)
   &SpawnRate(SpawnRate=20, Use Spawn Probability=false)
  ]
  particleSpawnScript [
   /Niagara/Modules/Spawn/Initialization/InitializeParticle(
    Mesh Scale=&VectorFromFloat(Value=3)
    Module.Write Color=true
    Module.Write Lifetime=true
    Module.Write Mass=true
    Module.Write Material Random=true
    Module.Write Position=false
    Module.Write Scale=true
    Module.Write SpriteRotation=false
    Module.Write SpriteSize=true
    Position=&SimulationPosition()
    Sprite Size=<1000 1000>
   )
   &SystemLocation(MaskSpawnGroup=false, Offset=$System.Position)
   /Niagara/Modules/Spawn/Location/ConeLocation(
    Cone Angle=0.1
    Cone Axis:Vector3f=$System.Angle
    Cone Length=0
    Cone Origin:NiagaraPosition=<0 0 0>
    MaskSpawnGroup=false
    Module.Override Randomness=false
    Offset:Vector3f=$System.Position
    Point Distribution=0.6
   ) {enabled=false, module=ConeLocation002}
  ]
  particleUpdateScript [
   &SkeletalMeshLocation(
    Apply Sampled Orientation to Mesh Particles=true
    Bone Sampling Mode=::NewEnumerator4
    Inherit Velocity=false
    MaskSpawnGroup=false
    Sampled Position Offset:NiagaraPosition=$Particles.Initial.Position
    Sampled Position Offset Coordinate Space:ENiagaraMeshTransforms=::NewEnumerator3
    Skeletal Mesh=$User.Skeletal Mesh
   )
   &ParticleState()
   &SolveForcesAndVelocity(Clamp Velocity=false, Limit Acceleration=false)
   /SetVariables(Particles.MeshOrientation=&OrientMeshToVector(Look At Direction=<-1 -0.84 -0.61>, Reference Axis Vector=<1 0 0>)) {enabled=false, isSetParameters=true}
   /SetVariables(
    Particles.MeshOrientation001:Quat4f=&OrientMeshToVector(
     Look At Direction=&RotateVectorByQuaternion(Vector To Rotate=$System.Angle, Quaternion=$Output.SkeletalMeshLocation.SampledRotation)
     Reference Axis Vector=<1 0 0>
    )
    Particles.moveColor:LinearColor=<rgba 1 1 1 0.8>
   ) {isSetParameters=true}
  ]
  rendererClasses [&NiagaraSpriteRendererProperties, &NiagaraMeshRendererProperties]
  renderers [
   {rendererClass=&NiagaraSpriteRendererProperties}
   {rendererClass=&NiagaraMeshRendererProperties, rendererIndex=1}
   {rendererClass=&NiagaraMeshRendererProperties, rendererIndex=2}
  ]
 }
 Sparkles {
  emitterUpdateScript [
   &EmitterState()
   &SpawnRate(Spawn Probability=0.1, SpawnRate=50, Use Spawn Probability=true)
   &SpawnRate(Spawn Probability=0.25, SpawnRate=2000, Use Spawn Probability=true) {enabled=false, module=SpawnRate001}
  ]
  particleSpawnScript [
   &InitializeParticle(
    Color=&HueShiftLinearColor(
     Hue=&RandomRangeFloat(Minimum=-0.01, Maximum=0.01)
     Saturation=&RandomRangeFloat(Minimum=0.75, Maximum=1)
     Value=1
     Linear Color=&MakeLinearColorFromVectorAndFloat("Vector (RGB)"=<1 0.33 0.12>, "Float (Alpha)"=&RandomRangeFloat(Minimum=0.25, Maximum=1))
    )
    Lifetime=&UniformRangedFloat(Minimum=0.3, Maximum=1)
    Module.Write Color=true
    Module.Write Lifetime=true
    Module.Write Mass=true
    Module.Write Material Random=true
    Module.Write Position=false
    Module.Write Scale=false
    Module.Write SpriteRotation=true
    Module.Write SpriteSize=true
    Position=&SimulationPosition()
    Sprite Rotation=&UniformRangedFloat(Minimum=0, Maximum=360)
    Sprite Size=&Vector2DFromFloat(Value=&UniformRangedFloat(Minimum=2, Maximum=6))
   )
   &ConeLocation(
    Cone Angle=4.3455
    Cone Axis:Vector3f=<-2.8981 1.1429 1.4117>
    Cone Length=11.1534
    Cone Origin:NiagaraPosition=<0 0 0>
    MaskSpawnGroup=false
    Module.Override Randomness=false
    Offset:Vector3f=<28.0445 7.6737 6.7229>
    Point Distribution=0.5
   )
   /SetVariables(Particles.HornRTipPos:Vector3f=&Lerp_Vector(
     A=Linked() {linkedVariable {name=Particles.Position, type=NiagaraPosition}}
     B=<19.9 5.3 6>
     Alpha=&UniformAOrBFloat(A=0, Distribution Weight=0.15)
    )) {isSetParameters=true}
   &ConeLocation(
    Cone Angle=1.75
    Cone Axis:Vector3f=<-0.8984 -0.6894 -0.61>
    Cone Length=28
    Cone Origin:NiagaraPosition=<0 0 0>
    MaskSpawnGroup=false
    Module.Override Randomness=false
    Offset:Vector3f=<29.3 21.1 19.6>
    Point Distribution=0.62
   ) {module=ConeLocation002}
   /SetVariables(Particles.HornRBasePos:NiagaraPosition=$Particles.Position) {isSetParameters=true}
   &ConeLocation(
    Cone Angle=4.3455
    Cone Axis:Vector3f=<-2.8981 1.1429 -1.4117>
    Cone Length=11.1534
    Cone Origin:NiagaraPosition=<0 0 0>
    MaskSpawnGroup=false
    Module.Override Randomness=false
    Offset:Vector3f=<28.0445 7.6737 -6.7229>
    Point Distribution=0.5
   ) {module=ConeLocation003}
   /SetVariables(Particles.HornLTipPos:Vector3f=&Lerp_Vector(
     A=Linked() {linkedVariable {name=Particles.Position, type=NiagaraPosition}}
     B=<19.9 5.3 -6>
     Alpha=&UniformAOrBFloat(A=0, Distribution Weight=0.15)
    )) {isSetParameters=true}
   &ConeLocation(
    Cone Angle=1.75
    Cone Axis:Vector3f=<-0.8984 -0.6894 0.61>
    Cone Length=28
    Cone Origin:NiagaraPosition=<0 0 0>
    MaskSpawnGroup=false
    Module.Override Randomness=false
    Offset:Vector3f=<29.3 21.1 -19.6>
    Point Distribution=0.62
   ) {module=ConeLocation001}
   /SetVariables(Particles.HornLBasePos:NiagaraPosition=$Particles.Position) {isSetParameters=true}
   &CylinderLocation(
    Cylinder Height=6.5677
    Cylinder Origin:Vector3f=<6.102 11.419 22.241>
    Cylinder Radius=3
    MaskSpawnGroup=false
    Module.Enable Lathe Profile=false
    Module.Override Randomness=false
    Module.Surface Only=false
   ) {enabled=false}
   /SetVariables(
    Particles.HornLPos:Vector3f=&Lerp_Vector(A=$Particles.HornLTipPos, B=$Particles.HornLBasePos, Alpha=&UniformAOrBFloat(A=0))
    Particles.HornRPos:Vector3f=&Lerp_Vector(A=$Particles.HornRTipPos, B=$Particles.HornRBasePos, Alpha=&UniformAOrBFloat(A=0))
   ) {isSetParameters=true}
   /SetVariables(Particles.Position=&Lerp_Vector(A=$Particles.HornRPos, B=$Particles.HornLPos, Alpha=&UniformAOrBFloat(A=0))) {isSetParameters=true}
  ]
  particleUpdateScript [
   &ParticleState()
   &SkeletalMeshLocation(
    Bone Sampling Mode=::NewEnumerator4
    Inherit Velocity=false
    MaskSpawnGroup=false
    Sampled Position Offset:NiagaraPosition=$Particles.Initial.Position
    Sampled Position Offset Coordinate Space:ENiagaraMeshTransforms=::NewEnumerator3
    Skeletal Mesh=$User.Skeletal Mesh
   )
   &ScaleColor(
    Scale Alpha=&FloatFromCurve(FloatCurve=DataInterface() {propertyValues {Curve {keys [<at -0.0092 0.0144>,<at 0.0967 0.9767>,<at 1.0011 0.002>]}}})
    Scale RGB=&VectorFromFloat(Value=&Lerp_Float(A=1, B=10, Alpha=&Power(a=&RandomRangeFloat(Minimum=0, Maximum=1), b=4)))
    ScaleA=true
    ScaleRGB=true
   )
   &ScaleSpriteSize(Scale Factor=&Vector2DFromFloat(Value=&FloatFromCurve(FloatCurve=DataInterface() {propertyValues {Curve {keys [<at -0.0046 0>,<at 0.1746 0.9952>,<at 1.0011 0.002>]}}})))
   &SpriteRotationRate(Rotation Rate=&RandomRangeFloat(Minimum=-100, Maximum=100))
  ]
  rendererClasses [&NiagaraSpriteRendererProperties]
  renderers [{rendererClass=&NiagaraSpriteRendererProperties}]
 }
}
summary {
 systemName=NS_Skin_PunkDevil_ExtremeGold
 emitters [
  {emitterName=Sparkles, rendererClasses [&NiagaraSpriteRendererProperties]}
  {
   bEnabled=false
   emitterName=NE_Empty
   rendererClasses [&NiagaraSpriteRendererProperties, &NiagaraMeshRendererProperties]
  }
 ]
 userVariables [User.Skeletal Mesh {
   type=NiagaraDataInterfaceSkeletalMesh
   defaultValue {
    struct=NiagaraExt_VariableValue_DataInterface
    value {dataInterface=&self:NiagaraDataInterfaceSkeletalMesh_0}
   }
  }]
}
